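# Animation Blueprint for the F_MED_RoyalDusk female player body cosmetic, adding secondary motion to skirt, side bling, purse zipper and choker.
# It copies the attached parent's pose, then uses weighted constraints to place attach and belt bones, offsetting backpack and cape for backblings.
# Thigh and neighbor-bone constraints feed spring-driven Anim Dynamics simulations.
# Leg constraint alpha variables blend corrective rotations and later thigh constraints on front skirt bones.
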AnimBlueprint /BRCosmetics/Characters/Player/Female/Medium/Bodies/F_MED_RoyalDusk/Meshes/F_MED_RoyalDusk_AnimBP
target_skeleton=&/F_MED_RoyalDusk_Skeleton
graphs [
 AnimGraph {nodes [
   n1 = Get R_leg constraintAlpha() -> (R_leg constraintAlpha:double)
   n2 = Get R_leg constraintAlpha()
   n3 = Get L_leg constraintAlpha() -> (L_leg constraintAlpha:double)
   n4 = "Constraint - attach_backpack (PPPPPPPP)"!(
    ComponentPose=Local To Component(LocalPose=Copy Pose From Mesh() {bUseAttachedParent=true})
    Alpha=1
    ConstraintWeights_0=0.3029
    ConstraintWeights_1:float=0.2492
    ConstraintWeights_2:float=0.162
    ConstraintWeights_3:float=0.1069
    ConstraintWeights_4:float=0.1068
    ConstraintWeights_5:float=0.0429
    ConstraintWeights_6:float=0.0172
    ConstraintWeights_7:float=0.0121
   ) -> (Pose) {
    alphaBoolBlend {blendOption=Linear}
    boneToModify {boneName=attach_backpack}
    constraintSetup [
     {targetBone {boneName=spine_05}}
     {targetBone {boneName=spine_03}}
     {targetBone {boneName=spine_04}}
     {targetBone {boneName=clavicle_l}}
     {targetBone {boneName=clavicle_r}}
     {targetBone {boneName=spine_02}}
     {targetBone {boneName=upperarm_twist_01_r}}
     {targetBone {boneName=upperarm_twist_01_l}}
    ]
    constraintWeights [1,1,1,1,1,1,1,1]
   }
   n5 = "Constraint - attach_cape (PPPPPP)"!(
    ComponentPose=@n4
    Alpha=1
    ConstraintWeights_0=0.7413
    ConstraintWeights_1:float=0.1094
    ConstraintWeights_2:float=0.0635
    ConstraintWeights_3:float=0.0497
    ConstraintWeights_4:float=0.0296
    ConstraintWeights_5:float=0.0065
   ) -> (Pose) {
    alphaBoolBlend {blendOption=Linear}
    boneToModify {boneName=attach_cape}
    constraintSetup [
     {targetBone {boneName=spine_05}}
     {targetBone {boneName=clavicle_l}}
     {targetBone {boneName=clavicle_r}}
     {targetBone {boneName=spine_04}}
     {targetBone {boneName=spine_03}}
     {targetBone {boneName=upperarm_l}}
    ]
    constraintWeights [1,1,1,1,1,1]
   }
   n6 = "Transform (Modify) Bone - Bone: attach_backpack"!(ComponentPose=@n5, bAlphaBoolEnabled:bool=Get Is Wearing Backbling(), Translation="0,-1,0", Rotation="0,0,0", Scale="1,1,1") -> (Pose) {
    alphaInputType=Bool
    translationMode=BMM_Additive
    translationSpace=BCS_BoneSpace
    alphaBoolBlend {blendOption=Linear}
    boneToModify {boneName=attach_backpack}
    translation {y=-1}
   }
   n7 = "Transform (Modify) Bone - Bone: attach_cape"!(
    ComponentPose=@n6
    bAlphaBoolEnabled:bool=Get Is Backbling A Cape() -> (Is Backbling A Cape:bool)
    Translation="0,-1.5,-2"
    Rotation="0,0,0"
    Scale="1,1,1"
   ) -> (Pose) {
    alphaInputType=Bool
    translationMode=BMM_Additive
    translationSpace=BCS_BoneSpace
    alphaBoolBlend {blendOption=Linear}
    boneToModify {boneName=attach_cape}
    translation {y=-1.5, z=-2}
   }
   n8 = "Constraint - dyn_coilBelt_A (PPPPP)"!(
    ComponentPose=@n7
    Alpha=1
    ConstraintWeights_0=0.8105
    ConstraintWeights_1:float=0.086
    ConstraintWeights_2:float=0.0573
    ConstraintWeights_3:float=0.0337
    ConstraintWeights_4:float=0.0126
   ) -> (Pose) {
    alphaBoolBlend {blendOption=Linear}
    boneToModify {boneName=dyn_coilBelt_A}
    constraintSetup [
     {targetBone {boneName=pelvis}}
     {targetBone {boneName=spine_01}}
     {targetBone {boneName=spine_02}}
     {targetBone {boneName=thigh_l}}
     {targetBone {boneName=thigh_r}}
    ]
    constraintWeights [1,1,1,1,1]
   }
   n9 = "Constraint - dyn_coilBelt_B (PPP)"!(ComponentPose=@n8, Alpha=1, ConstraintWeights_0=0.7286, ConstraintWeights_1:float=0.1424, ConstraintWeights_2:float=0.129) -> (Pose) {
    alphaBoolBlend {blendOption=Linear}
    boneToModify {boneName=dyn_coilBelt_B}
    constraintSetup [{targetBone {boneName=pelvis}}, {targetBone {boneName=spine_01}}, {targetBone {boneName=spine_02}}]
    constraintWeights [1,1,1]
   }
   n10 = "Constraint - dyn_coilBelt_H (PPPP)"!(
    ComponentPose=@n9
    Alpha=1
    ConstraintWeights_0=0.8687
    ConstraintWeights_1:float=0.0819
    ConstraintWeights_2:float=0.0275
    ConstraintWeights_3:float=0.022
   ) -> (Pose) {
    alphaBoolBlend {blendOption=Linear}
    boneToModify {boneName=dyn_coilBelt_H}
    constraintSetup [{targetBone {boneName=pelvis}}, {targetBone {boneName=spine_01}}, {targetBone {boneName=thigh_l}}, {targetBone {boneName=spine_02}}]
    constraintWeights [1,1,1,1]
   }
   n11 = "Constraint - dyn_coilBelt_I (PPPP)"!(
    ComponentPose=@n10
    Alpha=1
    ConstraintWeights_0=0.8493
    ConstraintWeights_1:float=0.0829
    ConstraintWeights_2:float=0.0429
    ConstraintWeights_3:float=0.0249
   ) -> (Pose) {
    alphaBoolBlend {blendOption=Linear}
    boneToModify {boneName=dyn_coilBelt_I}
    constraintSetup [{targetBone {boneName=pelvis}}, {targetBone {boneName=spine_01}}, {targetBone {boneName=thigh_l}}, {targetBone {boneName=spine_02}}]
    constraintWeights [1,1,1,1]
   }
   n12 = "Constraint - dyn_coilBelt_J (PPPP)"!(
    ComponentPose=@n11
    Alpha=1
    ConstraintWeights_0=0.8244
    ConstraintWeights_1:float=0.0791
    ConstraintWeights_2:float=0.0624
    ConstraintWeights_3:float=0.034
   ) -> (Pose) {
    alphaBoolBlend {blendOption=Linear}
    boneToModify {boneName=dyn_coilBelt_J}
    constraintSetup [{targetBone {boneName=pelvis}}, {targetBone {boneName=spine_01}}, {targetBone {boneName=thigh_l}}, {targetBone {boneName=spine_02}}]
    constraintWeights [1,1,1,1]
   }
   n13 = "Constraint - dyn_purseZipper (PPPPP)"!(
    ComponentPose=@n12
    Alpha=1
    ConstraintWeights_0=0.2289
    ConstraintWeights_1:float=0.5163
    ConstraintWeights_2:float=0.1683
    ConstraintWeights_3:float=0.0536
    ConstraintWeights_4:float=0.033
   ) -> (Pose) {
    alphaBoolBlend {blendOption=Linear}
    boneToModify {boneName=dyn_purseZipper}
    constraintSetup [
     {targetBone {boneName=spine_04}}
     {targetBone {boneName=spine_03}}
     {targetBone {boneName=spine_02}}
     {targetBone {boneName=spine_05}}
     {targetBone {boneName=spine_01}}
    ]
    constraintWeights [1,1,1,1,1]
   }
   n14 = "Constraint - dyn_sideBling_A_1 (PPP)"!(ComponentPose=@n13, Alpha=1, ConstraintWeights_0=0, ConstraintWeights_1:float=1, ConstraintWeights_2:float=0) -> (Pose) {
    alphaBoolBlend {blendOption=Linear}
    alphaScaleBiasClamp {scale=0.3}
    boneToModify {boneName=dyn_sideBling_A_1}
    constraintSetup [{targetBone {boneName=dyn_side_R_skirt_1}}, {targetBone {boneName=dyn_side_R_skirt_2}}, {targetBone {boneName=dyn_side_R_skirt_3}}]
    constraintWeights [1,1,1]
   }
   n15 = "Constraint - dyn_sideBling_A_2 (PPP)"!(ComponentPose=@n14, Alpha=1, ConstraintWeights_0=0, ConstraintWeights_1:float=1, ConstraintWeights_2:float=0) -> (Pose) {
    alphaBoolBlend {blendOption=Linear}
    alphaScaleBiasClamp {scale=0.3}
    boneToModify {boneName=dyn_sideBling_A_2}
    constraintSetup [{targetBone {boneName=dyn_side_R_skirt_1}}, {targetBone {boneName=dyn_side_R_skirt_2}}, {targetBone {boneName=dyn_side_R_skirt_3}}]
    constraintWeights [1,1,1]
   }
   n16 = "Constraint - dyn_sideBling_A_3 (PPP)"!(ComponentPose=@n15, Alpha=1, ConstraintWeights_0=0, ConstraintWeights_1:float=0, ConstraintWeights_2:float=1) -> (Pose) {
    alphaBoolBlend {blendOption=Linear}
    alphaScaleBiasClamp {scale=0.3}
    boneToModify {boneName=dyn_sideBling_A_3}
    constraintSetup [{targetBone {boneName=dyn_side_R_skirt_1}}, {targetBone {boneName=dyn_side_R_skirt_2}}, {targetBone {boneName=dyn_side_R_skirt_3}}]
    constraintWeights [1,1,1]
   }
   n17 = "Constraint - dyn_sideBling_B_1 (PPP)"!(ComponentPose=@n16, Alpha=1, ConstraintWeights_0=0, ConstraintWeights_1:float=1, ConstraintWeights_2:float=0) -> (Pose) {
    alphaBoolBlend {blendOption=Linear}
    alphaScaleBiasClamp {scale=0.3}
    boneToModify {boneName=dyn_sideBling_B_1}
    constraintSetup [{targetBone {boneName=dyn_side_R_skirt_1}}, {targetBone {boneName=dyn_side_R_skirt_2}}, {targetBone {boneName=dyn_side_R_skirt_3}}]
    constraintWeights [1,1,1]
   }
   n18 = "Constraint - dyn_sideBling_B_2 (PPP)"!(ComponentPose=@n17, Alpha=1, ConstraintWeights_0=0, ConstraintWeights_1:float=1, ConstraintWeights_2:float=0) -> (Pose) {
    alphaBoolBlend {blendOption=Linear}
    alphaScaleBiasClamp {scale=0.3}
    boneToModify {boneName=dyn_sideBling_B_2}
    constraintSetup [{targetBone {boneName=dyn_side_R_skirt_1}}, {targetBone {boneName=dyn_side_R_skirt_2}}, {targetBone {boneName=dyn_side_R_skirt_3}}]
    constraintWeights [1,1,1]
   }
   n19 = "Constraint - dyn_sideBling_B_3 (PPP)"!(ComponentPose=@n18, Alpha=1, ConstraintWeights_0=0, ConstraintWeights_1:float=0, ConstraintWeights_2:float=1) -> (Pose) {
    alphaBoolBlend {blendOption=Linear}
    alphaScaleBiasClamp {scale=0.3}
    boneToModify {boneName=dyn_sideBling_B_3}
    constraintSetup [{targetBone {boneName=dyn_side_R_skirt_1}}, {targetBone {boneName=dyn_side_R_skirt_2}}, {targetBone {boneName=dyn_side_R_skirt_3}}]
    constraintWeights [1,1,1]
   }
   n20 = Anim Dynamics - Bone: dyn_sideBling_A_1(ComponentPose=@n19, Alpha=1, ExternalForce="0,0,0") {
    angularDampingOverride=1
    angularSpringConstant=6000
    bAngularSpring=true
    bOverrideAngularDamping=true
    bUseGravityOverride=true
    gravityScale=0.35
    simulationSpace=RootRelative
    alphaBoolBlend {blendOption=Linear}
    boundBone {boneName=dyn_sideBling_A_1}
    componentLinearAccScale=<1 1 1>
    componentLinearVelScale=<1 1 1>
    gravityOverride {z=-1500}
    physicsBodyDefinitions [{
      boundBone {boneName=dyn_sideBling_A_1}
      boxExtents=<23 23 23>
      constraintSetup {angularLimitsMax=<0 50 3>, angularLimitsMin=<0 0 -3>, angularTarget=<1 0 0>}
      localJointOffset=<2 0 0>
     }]
   }
   n21 = Anim Dynamics - Bone: dyn_sideBling_A_2(ComponentPose=@n20, Alpha=1, ExternalForce="0,0,0") {
    angularDampingOverride=1
    angularSpringConstant=6000
    bAngularSpring=true
    bOverrideAngularDamping=true
    bUseGravityOverride=true
    gravityScale=0.35
    simulationSpace=RootRelative
    alphaBoolBlend {blendOption=Linear}
    boundBone {boneName=dyn_sideBling_A_2}
    componentLinearAccScale=<1 1 1>
    componentLinearVelScale=<1 1 1>
    gravityOverride {z=-1500}
    physicsBodyDefinitions [{
      boundBone {boneName=dyn_sideBling_A_2}
      boxExtents=<23 23 23>
      constraintSetup {angularLimitsMax=<0 50 3>, angularLimitsMin=<0 -2 -3>, angularTarget=<1 0 0>}
      localJointOffset=<2 0 0>
     }]
   }
   n22 = Anim Dynamics - Bone: dyn_sideBling_A_3(ComponentPose=@n21, Alpha=1, ExternalForce="0,0,0") {
    angularDampingOverride=1
    angularSpringConstant=6000
    bAngularSpring=true
    bOverrideAngularDamping=true
    bUseGravityOverride=true
    gravityScale=0.35
    simulationSpace=RootRelative
    alphaBoolBlend {blendOption=Linear}
    boundBone {boneName=dyn_sideBling_A_3}
    componentLinearAccScale=<1 1 1>
    componentLinearVelScale=<1 1 1>
    gravityOverride {z=-1500}
    physicsBodyDefinitions [{
      boundBone {boneName=dyn_sideBling_A_3}
      boxExtents=<23 23 23>
      constraintSetup {angularLimitsMax=<0 50 3>, angularLimitsMin=<0 5 -3>, angularTarget=<1 0 0>}
      localJointOffset=<2 0 0>
     }]
   }
   n23 = Anim Dynamics - Bone: dyn_sideBling_B_1(ComponentPose=@n22, Alpha=1, ExternalForce="0,0,0") {
    angularDampingOverride=1
    angularSpringConstant=6000
    bAngularSpring=true
    bOverrideAngularDamping=true
    bUseGravityOverride=true
    gravityScale=0.35
    simulationSpace=RootRelative
    alphaBoolBlend {blendOption=Linear}
    boundBone {boneName=dyn_sideBling_B_1}
    componentLinearAccScale=<1 1 1>
    componentLinearVelScale=<1 1 1>
    gravityOverride {z=-1500}
    physicsBodyDefinitions [{
      boundBone {boneName=dyn_sideBling_B_1}
      boxExtents=<29 29 29>
      constraintSetup {angularLimitsMax=<0 50 3>, angularLimitsMin=<0 0 -3>, angularTarget=<1 0 0>}
      localJointOffset=<2 0 0>
     }]
   }
   n24 = Anim Dynamics - Bone: dyn_sideBling_B_2(ComponentPose=@n23, Alpha=1, ExternalForce="0,0,0") {
    angularDampingOverride=1
    angularSpringConstant=6000
    bAngularSpring=true
    bOverrideAngularDamping=true
    bUseGravityOverride=true
    gravityScale=0.35
    simulationSpace=RootRelative
    alphaBoolBlend {blendOption=Linear}
    boundBone {boneName=dyn_sideBling_B_2}
    componentLinearAccScale=<1 1 1>
    componentLinearVelScale=<1 1 1>
    gravityOverride {z=-1500}
    physicsBodyDefinitions [{
      boundBone {boneName=dyn_sideBling_B_2}
      boxExtents=<29 29 29>
      constraintSetup {angularLimitsMax=<0 50 3>, angularLimitsMin=<0 -2 -3>, angularTarget=<1 0 0>}
      localJointOffset=<2 0 0>
     }]
   }
   n25 = Anim Dynamics - Bone: dyn_sideBling_B_3(ComponentPose=@n24, Alpha=1, ExternalForce="0,0,0") {
    angularDampingOverride=1
    angularSpringConstant=6000
    bAngularSpring=true
    bOverrideAngularDamping=true
    bUseGravityOverride=true
    gravityScale=0.35
    simulationSpace=RootRelative
    alphaBoolBlend {blendOption=Linear}
    boundBone {boneName=dyn_sideBling_B_3}
    componentLinearAccScale=<1 1 1>
    componentLinearVelScale=<1 1 1>
    gravityOverride {z=-1500}
    physicsBodyDefinitions [{
      boundBone {boneName=dyn_sideBling_B_3}
      boxExtents=<29 29 29>
      constraintSetup {angularLimitsMax=<0 50 3>, angularLimitsMin=<0 5 -3>, angularTarget=<1 0 0>}
      localJointOffset=<2 0 0>
     }]
   }
   n26 = "Transform (Modify) Bone - Bone: dyn_sideBling_A_2"(ComponentPose=@n25, Alpha=Get R_leg constraintAlpha(), Translation="0,0,0", Rotation="-10,20,9", Scale="1,1,1") {
    rotationMode=BMM_Additive
    alphaBoolBlend {blendOption=Linear}
    alphaScaleBiasClamp {bInterpResult=true, interpSpeedDecreasing=30, interpSpeedIncreasing=30}
    boneToModify {boneName=dyn_sideBling_A_2}
    rotation=<rot -10 20 9>
   }
   n27 = "Constraint - dyn_front_L_skirt_1 (P)"(ComponentPose=@n26, Alpha=1, ConstraintWeights_0=1) {
    alphaBoolBlend {blendOption=Linear}
    alphaScaleBiasClamp {scale=0.95}
    boneToModify {boneName=dyn_front_L_skirt_1}
    constraintSetup [{targetBone {boneName=thigh_l}}]
    constraintWeights [1]
   }
   n28 = "Constraint - dyn_front_L_skirt_2 (P)"(ComponentPose=@n27, Alpha=1, ConstraintWeights_0=1) {
    alphaBoolBlend {blendOption=Linear}
    alphaScaleBiasClamp {scale=0.5}
    boneToModify {boneName=dyn_front_L_skirt_2}
    constraintSetup [{targetBone {boneName=thigh_l}}]
    constraintWeights [1]
   }
   n29 = "Constraint - dyn_front_R_skirt_1 (P)"(ComponentPose=@n28, Alpha=1, ConstraintWeights_0=1) {
    alphaBoolBlend {blendOption=Linear}
    alphaScaleBiasClamp {scale=0.95}
    boneToModify {boneName=dyn_front_R_skirt_1}
    constraintSetup [{targetBone {boneName=thigh_r}}]
    constraintWeights [1]
   }
   n30 = "Constraint - dyn_front_R_skirt_2 (P)"(ComponentPose=@n29, Alpha=1, ConstraintWeights_0=1) {
    alphaBoolBlend {blendOption=Linear}
    alphaScaleBiasClamp {scale=0.5}
    boneToModify {boneName=dyn_front_R_skirt_2}
    constraintSetup [{targetBone {boneName=thigh_r}}]
    constraintWeights [1]
   }
   n31 = Anim Dynamics - Bone: dyn_front_L_skirt_1!(ComponentPose=@n30, Alpha=1, GravityOverride:Vector="0,0,-1500", ExternalForce="0,0,0") -> (Pose) {
    angularDampingOverride=0.9999
    angularSpringConstant=8000
    bAngularSpring=true
    bOverrideAngularDamping=true
    bUseGravityOverride=true
    gravityScale=0.35
    simulationSpace=RootRelative
    alphaBoolBlend {blendOption=Linear}
    boundBone {boneName=dyn_front_L_skirt_1}
    componentLinearAccScale=<1 1 1>
    componentLinearVelScale=<1 1 1>
    gravityOverride {z=-1500}
    physicsBodyDefinitions [{
      boundBone {boneName=dyn_front_L_skirt_1}
      boxExtents=<17 17 17>
      constraintSetup {angularLimitsMax=<10 50 3>, angularLimitsMin=<-10 0 -3>, angularTarget=<1 0 0>}
      localJointOffset=<3 0 0>
     }]
   }
   n32 = Anim Dynamics - Bone: dyn_front_R_skirt_1!(ComponentPose=@n31, Alpha=1, GravityOverride:Vector="0,0,-1500", ExternalForce="0,0,0") -> (Pose) {
    angularDampingOverride=0.9999
    angularSpringConstant=8000
    bAngularSpring=true
    bOverrideAngularDamping=true
    bUseGravityOverride=true
    gravityScale=0.35
    simulationSpace=RootRelative
    alphaBoolBlend {blendOption=Linear}
    boundBone {boneName=dyn_front_R_skirt_1}
    componentLinearAccScale=<1 1 1>
    componentLinearVelScale=<1 1 1>
    gravityOverride {z=-1500}
    physicsBodyDefinitions [{
      boundBone {boneName=dyn_front_R_skirt_1}
      boxExtents=<17 17 17>
      constraintSetup {angularLimitsMax=<10 50 3>, angularLimitsMin=<-10 0 -3>, angularTarget=<1 0 0>}
      localJointOffset=<3 0 0>
     }]
   }
   n33 = Anim Dynamics - Bone: dyn_front_L_skirt_2!(ComponentPose=@n32, Alpha=1, GravityOverride:Vector="0,0,-1500", ExternalForce="0,0,0") -> (Pose) {
    angularDampingOverride=0.9999
    angularSpringConstant=8000
    bAngularSpring=true
    bOverrideAngularDamping=true
    bUseGravityOverride=true
    gravityScale=0.35
    simulationSpace=RootRelative
    alphaBoolBlend {blendOption=Linear}
    boundBone {boneName=dyn_front_L_skirt_2}
    componentLinearAccScale=<1 1 1>
    componentLinearVelScale=<1 1 1>
    gravityOverride {z=-1500}
    physicsBodyDefinitions [{
      boundBone {boneName=dyn_front_L_skirt_2}
      boxExtents=<17 17 17>
      constraintSetup {angularLimitsMax=<10 50 3>, angularLimitsMin=<-10 -2 -3>, angularTarget=<1 0 0>}
      localJointOffset=<3 0 0>
     }]
   }
   n34 = Anim Dynamics - Bone: dyn_front_R_skirt_2!(ComponentPose=@n33, Alpha=1, GravityOverride:Vector="0,0,-1500", ExternalForce="0,0,0") -> (Pose) {
    angularDampingOverride=0.9999
    angularSpringConstant=8000
    bAngularSpring=true
    bOverrideAngularDamping=true
    bUseGravityOverride=true
    gravityScale=0.35
    simulationSpace=RootRelative
    alphaBoolBlend {blendOption=Linear}
    boundBone {boneName=dyn_front_R_skirt_2}
    componentLinearAccScale=<1 1 1>
    componentLinearVelScale=<1 1 1>
    gravityOverride {z=-1500}
    physicsBodyDefinitions [{
      boundBone {boneName=dyn_front_R_skirt_2}
      boxExtents=<17 17 17>
      constraintSetup {angularLimitsMax=<10 50 3>, angularLimitsMin=<-10 -2 -3>, angularTarget=<1 0 0>}
      localJointOffset=<3 0 0>
     }]
   }
   n35 = Anim Dynamics - Bone: dyn_front_L_skirt_3!(ComponentPose=@n34, Alpha=1, GravityOverride:Vector="0,0,-1500", ExternalForce="0,0,0") -> (Pose) {
    angularDampingOverride=0.9999
    angularSpringConstant=8000
    bAngularSpring=true
    bOverrideAngularDamping=true
    bUseGravityOverride=true
    gravityScale=0.35
    simulationSpace=RootRelative
    alphaBoolBlend {blendOption=Linear}
    boundBone {boneName=dyn_front_L_skirt_3}
    componentLinearAccScale=<1 1 1>
    componentLinearVelScale=<1 1 1>
    gravityOverride {z=-1500}
    physicsBodyDefinitions [{
      boundBone {boneName=dyn_front_L_skirt_3}
      boxExtents=<17 17 17>
      constraintSetup {angularLimitsMax=<10 50 3>, angularLimitsMin=<-10 -2 -3>, angularTarget=<1 0 0>}
      localJointOffset=<3 0 0>
     }]
   }
   n36 = Anim Dynamics - Bone: dyn_front_R_skirt_3!(ComponentPose=@n35, Alpha=1, GravityOverride:Vector="0,0,-1500", ExternalForce="0,0,0") -> (Pose) {
    angularDampingOverride=0.9999
    angularSpringConstant=8000
    bAngularSpring=true
    bOverrideAngularDamping=true
    bUseGravityOverride=true
    gravityScale=0.35
    simulationSpace=RootRelative
    alphaBoolBlend {blendOption=Linear}
    boundBone {boneName=dyn_front_R_skirt_3}
    componentLinearAccScale=<1 1 1>
    componentLinearVelScale=<1 1 1>
    gravityOverride {z=-1500}
    physicsBodyDefinitions [{
      boundBone {boneName=dyn_front_R_skirt_3}
      boxExtents=<17 17 17>
      constraintSetup {angularLimitsMax=<10 50 3>, angularLimitsMin=<-10 -2 -3>, angularTarget=<1 0 0>}
      localJointOffset=<3 0 0>
     }]
   }
   n37 = "Constraint - dyn_front_C_skirt_1 (PP)"!(ComponentPose=@n36, Alpha=1, ConstraintWeights_0=0.5, ConstraintWeights_1:float=0.5) -> (Pose) {
    alphaBoolBlend {blendOption=Linear}
    alphaScaleBiasClamp {scale=0.8}
    boneToModify {boneName=dyn_front_C_skirt_1}
    constraintSetup [{targetBone {boneName=dyn_front_L_skirt_2}}, {targetBone {boneName=dyn_front_R_skirt_2}}]
    constraintWeights [1,1]
   }
   n38 = Anim Dynamics - Bone: dyn_front_C_skirt_1!(ComponentPose=@n37, Alpha=1, GravityOverride:Vector="0,0,-1500", ExternalForce="0,0,0") -> (Pose) {
    angularDampingOverride=0.9999
    angularSpringConstant=2000
    bAngularSpring=true
    bOverrideAngularDamping=true
    bUseGravityOverride=true
    gravityScale=0.35
    simulationSpace=RootRelative
    alphaBoolBlend {blendOption=Linear}
    boundBone {boneName=dyn_front_C_skirt_1}
    componentLinearAccScale=<1 1 1>
    componentLinearVelScale=<1 1 1>
    gravityOverride {z=-1500}
    physicsBodyDefinitions [{
      boundBone {boneName=dyn_front_C_skirt_1}
      boxExtents=<25 25 25>
      constraintSetup {angularLimitsMax=<10 25 45>, angularLimitsMin=<-10 -20 -45>, angularTarget=<1 0 0>}
      localJointOffset=<2 0 0>
     }]
   }
   n39 = "Constraint - dyn_front_C_skirt_2 (PP)"!(ComponentPose=@n38, Alpha=1, ConstraintWeights_0=0.5, ConstraintWeights_1:float=0.5) -> (Pose) {
    alphaBoolBlend {blendOption=Linear}
    alphaScaleBiasClamp {scale=0.8}
    boneToModify {boneName=dyn_front_C_skirt_2}
    constraintSetup [{targetBone {boneName=dyn_front_L_skirt_3}}, {targetBone {boneName=dyn_front_R_skirt_3}}]
    constraintWeights [1,1]
   }
   n40 = Anim Dynamics - Bone: dyn_front_C_skirt_2!(ComponentPose=@n39, Alpha=1, GravityOverride:Vector="0,0,-1500", ExternalForce="0,0,0") -> (Pose) {
    angularDampingOverride=0.9999
    angularSpringConstant=2000
    bAngularSpring=true
    bOverrideAngularDamping=true
    bUseGravityOverride=true
    gravityScale=0.35
    simulationSpace=RootRelative
    alphaBoolBlend {blendOption=Linear}
    boundBone {boneName=dyn_front_C_skirt_2}
    componentLinearAccScale=<1 1 1>
    componentLinearVelScale=<1 1 1>
    gravityOverride {z=-1500}
    physicsBodyDefinitions [{
      boundBone {boneName=dyn_front_C_skirt_2}
      boxExtents=<25 25 25>
      constraintSetup {angularLimitsMax=<10 25 45>, angularLimitsMin=<-10 -20 -45>, angularTarget=<1 0 0>}
      localJointOffset=<2 0 0>
     }]
   }
   n41 = "Transform (Modify) Bone - Bone: dyn_front_C_skirt_master"(
    ComponentPose=@n40
    Alpha=Get eitherLegConstraintAlpha() -> (eitherLegConstraintAlpha:double)
    Translation="0,0,0"
    Rotation="-10,0,0"
    Scale="1,1,1"
   ) {
    rotationMode=BMM_Additive
    rotationSpace=BCS_BoneSpace
    scaleMode=BMM_Additive
    scaleSpace=BCS_BoneSpace
    translationMode=BMM_Additive
    translationSpace=BCS_BoneSpace
    alphaBoolBlend {blendOption=Linear}
    boneToModify {boneName=dyn_front_C_skirt_master}
    rotation {pitch=-10}
   }
   n42 = Anim Dynamics - Bone: dyn_back_L_skirt_1!(SimulationSpace:AnimPhysSimSpaceType Enum, ComponentPose=@n41, Alpha=1, GravityOverride:Vector="0,0,-1500", ExternalForce="0,0,0") -> (Pose) {
    angularDampingOverride=1
    angularSpringConstant=5000
    bAngularSpring=true
    bOverrideAngularDamping=true
    bUseGravityOverride=true
    gravityScale=0.35
    simulationSpace=RootRelative
    alphaBoolBlend {blendOption=Linear}
    boundBone {boneName=dyn_back_L_skirt_1}
    componentLinearAccScale=<1 1 1>
    componentLinearVelScale=<1 1 1>
    gravityOverride {z=-1500}
    physicsBodyDefinitions [{
      boundBone {boneName=dyn_back_L_skirt_1}
      boxExtents=<18 18 18>
      constraintSetup {angularLimitsMax=<20 50 15>, angularLimitsMin=<-20 0 -15>, angularTarget=<1 0 0>}
      localJointOffset=<2 0 0>
     }]
   }
   n43 = Anim Dynamics - Bone: dyn_back_C_skirt_1!(SimulationSpace:AnimPhysSimSpaceType Enum, ComponentPose=@n42, Alpha=1, GravityOverride:Vector="0,0,-1500", ExternalForce="0,0,0") -> (Pose) {
    angularDampingOverride=0.9999
    angularSpringConstant=5000
    bAngularSpring=true
    bOverrideAngularDamping=true
    bUseGravityOverride=true
    gravityScale=0.35
    simulationSpace=RootRelative
    alphaBoolBlend {blendOption=Linear}
    boundBone {boneName=dyn_back_C_skirt_1}
    componentLinearAccScale=<1 1 1>
    componentLinearVelScale=<1 1 1>
    gravityOverride {z=-1500}
    physicsBodyDefinitions [{
      boundBone {boneName=dyn_back_C_skirt_1}
      boxExtents=<18 18 18>
      constraintSetup {angularLimitsMax=<20 30 10>, angularLimitsMin=<-20 -10 -10>, angularTarget=<1 0 0>}
      localJointOffset=<2 0 0>
     }]
   }
   n44 = Anim Dynamics - Bone: dyn_back_R_skirt_1!(SimulationSpace:AnimPhysSimSpaceType Enum, ComponentPose=@n43, Alpha=1, GravityOverride:Vector="0,0,-1500", ExternalForce="0,0,0") -> (Pose) {
    angularDampingOverride=1
    angularSpringConstant=5000
    bAngularSpring=true
    bOverrideAngularDamping=true
    bUseGravityOverride=true
    gravityScale=0.35
    simulationSpace=RootRelative
    alphaBoolBlend {blendOption=Linear}
    boundBone {boneName=dyn_back_R_skirt_1}
    componentLinearAccScale=<1 1 1>
    componentLinearVelScale=<1 1 1>
    gravityOverride {z=-1500}
    physicsBodyDefinitions [{
      boundBone {boneName=dyn_back_R_skirt_1}
      boxExtents=<18 18 18>
      constraintSetup {angularLimitsMax=<20 50 15>, angularLimitsMin=<-20 0 -15>, angularTarget=<1 0 0>}
      localJointOffset=<2 0 0>
     }]
   }
   n45 = Anim Dynamics - Bone: dyn_back_L_skirt_2!(SimulationSpace:AnimPhysSimSpaceType Enum, ComponentPose=@n44, Alpha=1, GravityOverride:Vector="0,0,-1500", ExternalForce="0,0,0") -> (Pose) {
    angularDampingOverride=1
    angularSpringConstant=5000
    bAngularSpring=true
    bOverrideAngularDamping=true
    bUseGravityOverride=true
    gravityScale=0.35
    simulationSpace=RootRelative
    alphaBoolBlend {blendOption=Linear}
    boundBone {boneName=dyn_back_L_skirt_2}
    componentLinearAccScale=<1 1 1>
    componentLinearVelScale=<1 1 1>
    gravityOverride {z=-1500}
    physicsBodyDefinitions [{
      boundBone {boneName=dyn_back_L_skirt_2}
      boxExtents=<18 18 18>
      constraintSetup {angularLimitsMax=<20 50 15>, angularLimitsMin=<-20 0 -15>, angularTarget=<1 0 0>}
      localJointOffset=<2 0 0>
     }]
   }
   n46 = Anim Dynamics - Bone: dyn_back_C_skirt_2!(SimulationSpace:AnimPhysSimSpaceType Enum, ComponentPose=@n45, Alpha=1, GravityOverride:Vector="0,0,-1500", ExternalForce="0,0,0") -> (Pose) {
    angularDampingOverride=0.9999
    angularSpringConstant=5000
    bAngularSpring=true
    bOverrideAngularDamping=true
    bUseGravityOverride=true
    gravityScale=0.35
    simulationSpace=RootRelative
    alphaBoolBlend {blendOption=Linear}
    boundBone {boneName=dyn_back_C_skirt_2}
    componentLinearAccScale=<1 1 1>
    componentLinearVelScale=<1 1 1>
    gravityOverride {z=-1500}
    physicsBodyDefinitions [{
      boundBone {boneName=dyn_back_C_skirt_2}
      boxExtents=<18 18 18>
      constraintSetup {angularLimitsMax=<20 30 10>, angularLimitsMin=<-20 -15 -10>, angularTarget=<1 0 0>}
      localJointOffset=<2 0 0>
     }]
   }
   n47 = Anim Dynamics - Bone: dyn_back_R_skirt_2!(SimulationSpace:AnimPhysSimSpaceType Enum, ComponentPose=@n46, Alpha=1, GravityOverride:Vector="0,0,-1500", ExternalForce="0,0,0") -> (Pose) {
    angularDampingOverride=1
    angularSpringConstant=5000
    bAngularSpring=true
    bOverrideAngularDamping=true
    bUseGravityOverride=true
    gravityScale=0.35
    simulationSpace=RootRelative
    alphaBoolBlend {blendOption=Linear}
    boundBone {boneName=dyn_back_R_skirt_2}
    componentLinearAccScale=<1 1 1>
    componentLinearVelScale=<1 1 1>
    gravityOverride {z=-1500}
    physicsBodyDefinitions [{
      boundBone {boneName=dyn_back_R_skirt_2}
      boxExtents=<18 18 18>
      constraintSetup {angularLimitsMax=<20 50 15>, angularLimitsMin=<-20 0 -15>, angularTarget=<1 0 0>}
      localJointOffset=<2 0 0>
     }]
   }
   n48 = Anim Dynamics - Bone: dyn_back_L_skirt_3!(SimulationSpace:AnimPhysSimSpaceType Enum, ComponentPose=@n47, Alpha=1, GravityOverride:Vector="0,0,-1500", ExternalForce="0,0,0") -> (Pose) {
    angularDampingOverride=1
    angularSpringConstant=5000
    bAngularSpring=true
    bOverrideAngularDamping=true
    bUseGravityOverride=true
    gravityScale=0.35
    simulationSpace=RootRelative
    alphaBoolBlend {blendOption=Linear}
    boundBone {boneName=dyn_back_L_skirt_3}
    componentLinearAccScale=<1 1 1>
    componentLinearVelScale=<1 1 1>
    gravityOverride {z=-1500}
    physicsBodyDefinitions [{
      boundBone {boneName=dyn_back_L_skirt_3}
      boxExtents=<18 18 18>
      constraintSetup {angularLimitsMax=<20 50 15>, angularLimitsMin=<-20 0 -15>, angularTarget=<1 0 0>}
      localJointOffset=<2 0 0>
     }]
   }
   n49 = Anim Dynamics - Bone: dyn_back_C_skirt_3!(SimulationSpace:AnimPhysSimSpaceType Enum, ComponentPose=@n48, Alpha=1, GravityOverride:Vector="0,0,-1500", ExternalForce="0,0,0") -> (Pose) {
    angularDampingOverride=0.9999
    angularSpringConstant=5000
    bAngularSpring=true
    bOverrideAngularDamping=true
    bUseGravityOverride=true
    gravityScale=0.35
    simulationSpace=RootRelative
    alphaBoolBlend {blendOption=Linear}
    boundBone {boneName=dyn_back_C_skirt_3}
    componentLinearAccScale=<1 1 1>
    componentLinearVelScale=<1 1 1>
    gravityOverride {z=-1500}
    physicsBodyDefinitions [{
      boundBone {boneName=dyn_back_C_skirt_3}
      boxExtents=<18 18 18>
      constraintSetup {angularLimitsMax=<20 30 10>, angularLimitsMin=<-20 -20 -10>, angularTarget=<1 0 0>}
      localJointOffset=<2 0 0>
     }]
   }
   n50 = Anim Dynamics - Bone: dyn_back_R_skirt_3!(SimulationSpace:AnimPhysSimSpaceType Enum, ComponentPose=@n49, Alpha=1, GravityOverride:Vector="0,0,-1500", ExternalForce="0,0,0") -> (Pose) {
    angularDampingOverride=1
    angularSpringConstant=5000
    bAngularSpring=true
    bOverrideAngularDamping=true
    bUseGravityOverride=true
    gravityScale=0.35
    simulationSpace=RootRelative
    alphaBoolBlend {blendOption=Linear}
    boundBone {boneName=dyn_back_R_skirt_3}
    componentLinearAccScale=<1 1 1>
    componentLinearVelScale=<1 1 1>
    gravityOverride {z=-1500}
    physicsBodyDefinitions [{
      boundBone {boneName=dyn_back_R_skirt_3}
      boxExtents=<18 18 18>
      constraintSetup {angularLimitsMax=<20 50 15>, angularLimitsMin=<-20 0 -15>, angularTarget=<1 0 0>}
      localJointOffset=<2 0 0>
     }]
   }
   n51 = "Constraint - dyn_side_L_skirt_master (P)"(ComponentPose=@n50, Alpha=1, ConstraintWeights_0=1) {
    alphaBoolBlend {blendOption=Linear}
    boneToModify {boneName=dyn_side_L_skirt_master}
    constraintSetup [{targetBone {boneName=thigh_l}}]
    constraintWeights [1]
   }
   n52 = "Constraint - dyn_side_L_skirt_2 (P)"(ComponentPose=@n51, Alpha=1, ConstraintWeights_0=1) {
    alphaBoolBlend {blendOption=Linear}
    boneToModify {boneName=dyn_side_L_skirt_2}
    constraintSetup [{targetBone {boneName=thigh_l}}]
    constraintWeights [1]
   }
   n53 = "Constraint - dyn_side_L_skirt_3 (P)"(ComponentPose=@n52, Alpha=1, ConstraintWeights_0=1) {
    alphaBoolBlend {blendOption=Linear}
    boneToModify {boneName=dyn_side_L_skirt_3}
    constraintSetup [{targetBone {boneName=thigh_l}}]
    constraintWeights [1]
   }
   n54 = "Constraint - dyn_side_R_skirt_master (P)"(ComponentPose=@n53, Alpha=1, ConstraintWeights_0=1) {
    alphaBoolBlend {blendOption=Linear}
    boneToModify {boneName=dyn_side_R_skirt_master}
    constraintSetup [{targetBone {boneName=thigh_r}}]
    constraintWeights [1]
   }
   n55 = "Constraint - dyn_side_R_skirt_2 (P)"(ComponentPose=@n54, Alpha=1, ConstraintWeights_0=1) {
    alphaBoolBlend {blendOption=Linear}
    boneToModify {boneName=dyn_side_R_skirt_2}
    constraintSetup [{targetBone {boneName=thigh_r}}]
    constraintWeights [1]
   }
   n56 = "Constraint - dyn_side_R_skirt_3 (P)"(ComponentPose=@n55, Alpha=1, ConstraintWeights_0=1) {
    alphaBoolBlend {blendOption=Linear}
    boneToModify {boneName=dyn_side_R_skirt_3}
    constraintSetup [{targetBone {boneName=thigh_r}}]
    constraintWeights [1]
   }
   n57 = Anim Dynamics - Bone: dyn_side_L_skirt_1!(SimulationSpace:AnimPhysSimSpaceType Enum, ComponentPose=@n56, Alpha=1, GravityOverride:Vector="0,0,-1500", ExternalForce="0,0,0") -> (Pose) {
    angularDampingOverride=0.9999
    angularSpringConstant=8000
    bAngularSpring=true
    bOverrideAngularDamping=true
    bUseGravityOverride=true
    gravityScale=0.35
    simulationSpace=RootRelative
    alphaBoolBlend {blendOption=Linear}
    boundBone {boneName=dyn_side_L_skirt_1}
    componentLinearAccScale=<1 1 1>
    componentLinearVelScale=<1 1 1>
    gravityOverride {z=-1500}
    physicsBodyDefinitions [{
      boundBone {boneName=dyn_side_L_skirt_1}
      boxExtents=<17 17 17>
      constraintSetup {angularLimitsMax=<0 50 0>, angularLimitsMin=<0 0 0>, angularTarget=<1 0 0>}
      localJointOffset=<3 0 0>
     }]
   }
   n58 = Anim Dynamics - Bone: dyn_side_L_skirt_2!(SimulationSpace:AnimPhysSimSpaceType Enum, ComponentPose=@n57, Alpha=1, GravityOverride:Vector="0,0,-1500", ExternalForce="0,0,0") -> (Pose) {
    angularDampingOverride=0.9999
    angularSpringConstant=8000
    bAngularSpring=true
    bOverrideAngularDamping=true
    bUseGravityOverride=true
    gravityScale=0.35
    simulationSpace=RootRelative
    alphaBoolBlend {blendOption=Linear}
    boundBone {boneName=dyn_side_L_skirt_2}
    componentLinearAccScale=<1 1 1>
    componentLinearVelScale=<1 1 1>
    gravityOverride {z=-1500}
    physicsBodyDefinitions [{
      boundBone {boneName=dyn_side_L_skirt_2}
      boxExtents=<17 17 17>
      constraintSetup {angularLimitsMax=<0 50 0>, angularLimitsMin=<0 0 0>, angularTarget=<1 0 0>}
      localJointOffset=<3 0 0>
     }]
   }
   n59 = Anim Dynamics - Bone: dyn_side_L_skirt_3!(SimulationSpace:AnimPhysSimSpaceType Enum, ComponentPose=@n58, Alpha=1, GravityOverride:Vector="0,0,-1500", ExternalForce="0,0,0") -> (Pose) {
    angularDampingOverride=0.9999
    angularSpringConstant=8000
    bAngularSpring=true
    bOverrideAngularDamping=true
    bUseGravityOverride=true
    gravityScale=0.35
    simulationSpace=RootRelative
    alphaBoolBlend {blendOption=Linear}
    boundBone {boneName=dyn_side_L_skirt_3}
    componentLinearAccScale=<1 1 1>
    componentLinearVelScale=<1 1 1>
    gravityOverride {z=-1500}
    physicsBodyDefinitions [{
      boundBone {boneName=dyn_side_L_skirt_3}
      boxExtents=<17 17 17>
      constraintSetup {angularLimitsMax=<0 50 0>, angularLimitsMin=<0 0 0>, angularTarget=<1 0 0>}
      localJointOffset=<3 0 0>
     }]
   }
   n60 = Anim Dynamics - Bone: dyn_side_R_skirt_1!(SimulationSpace:AnimPhysSimSpaceType Enum, ComponentPose=@n59, Alpha=1, GravityOverride:Vector="0,0,-1500", ExternalForce="0,0,0") -> (Pose) {
    angularDampingOverride=0.9999
    angularSpringConstant=8000
    bAngularSpring=true
    bOverrideAngularDamping=true
    bUseGravityOverride=true
    gravityScale=0.35
    simulationSpace=RootRelative
    alphaBoolBlend {blendOption=Linear}
    boundBone {boneName=dyn_side_R_skirt_1}
    componentLinearAccScale=<1 1 1>
    componentLinearVelScale=<1 1 1>
    gravityOverride {z=-1500}
    physicsBodyDefinitions [{
      boundBone {boneName=dyn_side_R_skirt_1}
      boxExtents=<17 17 17>
      constraintSetup {angularLimitsMax=<0 35 0>, angularLimitsMin=<0 0 0>, angularTarget=<1 0 0>}
      localJointOffset=<3 0 0>
     }]
   }
   n61 = Anim Dynamics - Bone: dyn_side_R_skirt_2!(SimulationSpace:AnimPhysSimSpaceType Enum, ComponentPose=@n60, Alpha=1, GravityOverride:Vector="0,0,-1500", ExternalForce="0,0,0") -> (Pose) {
    angularDampingOverride=0.9999
    angularSpringConstant=8000
    bAngularSpring=true
    bOverrideAngularDamping=true
    bUseGravityOverride=true
    gravityScale=0.35
    simulationSpace=RootRelative
    alphaBoolBlend {blendOption=Linear}
    boundBone {boneName=dyn_side_R_skirt_2}
    componentLinearAccScale=<1 1 1>
    componentLinearVelScale=<1 1 1>
    gravityOverride {z=-1500}
    physicsBodyDefinitions [{
      boundBone {boneName=dyn_side_R_skirt_2}
      boxExtents=<17 17 17>
      constraintSetup {angularLimitsMax=<0 35 0>, angularLimitsMin=<0 0 0>, angularTarget=<1 0 0>}
      localJointOffset=<3 0 0>
     }]
   }
   n62 = Anim Dynamics - Bone: dyn_side_R_skirt_3!(SimulationSpace:AnimPhysSimSpaceType Enum, ComponentPose=@n61, Alpha=1, GravityOverride:Vector="0,0,-1500", ExternalForce="0,0,0") -> (Pose) {
    angularDampingOverride=0.9999
    angularSpringConstant=8000
    bAngularSpring=true
    bOverrideAngularDamping=true
    bUseGravityOverride=true
    gravityScale=0.35
    simulationSpace=RootRelative
    alphaBoolBlend {blendOption=Linear}
    boundBone {boneName=dyn_side_R_skirt_3}
    componentLinearAccScale=<1 1 1>
    componentLinearVelScale=<1 1 1>
    gravityOverride {z=-1500}
    physicsBodyDefinitions [{
      boundBone {boneName=dyn_side_R_skirt_3}
      boxExtents=<17 17 17>
      constraintSetup {angularLimitsMax=<0 50 0>, angularLimitsMin=<0 0 0>, angularTarget=<1 0 0>}
      localJointOffset=<3 0 0>
     }]
   }
   n63 = Anim Dynamics - Bone: dyn_purseZipper(ComponentPose=@n62, Alpha=1, ExternalForce="0,0,0") {
    angularDampingOverride=0.9999
    angularSpringConstant=2000
    bOverrideAngularDamping=true
    bUseGravityOverride=true
    gravityScale=0.35
    numSolverIterationsPreUpdate=2
    simulationSpace=RootRelative
    alphaBoolBlend {blendOption=Linear}
    boundBone {boneName=dyn_purseZipper}
    componentLinearAccScale=<1 1 1>
    componentLinearVelScale=<1 1 1>
    gravityOverride {z=-980}
    physicsBodyDefinitions [{
      boundBone {boneName=dyn_purseZipper}
      boxExtents=<7 7 7>
      constraintSetup {angularLimitsMax=<80 80 20>, angularLimitsMin=<-80 -80 -80>, angularTarget=<1 0 0>}
      localJointOffset=<2 0 0>
     }]
   }
   n64 = Anim Dynamics - Bone: dyn_chokerDiamond(ComponentPose=@n63, Alpha=1, ExternalForce="0,0,0") {
    angularDampingOverride=0.9999
    angularSpringConstant=2000
    bAngularSpring=true
    bOverrideAngularDamping=true
    bUseGravityOverride=true
    gravityScale=0.35
    numSolverIterationsPreUpdate=3
    simulationSpace=RootRelative
    alphaBoolBlend {blendOption=Linear}
    boundBone {boneName=dyn_chokerDiamond}
    componentLinearAccScale=<1 1 1>
    componentLinearVelScale=<1 1 1>
    gravityOverride {z=-980}
    physicsBodyDefinitions [{
      boundBone {boneName=dyn_chokerDiamond}
      boxExtents=<15 15 15>
      constraintSetup {angularLimitsMax=<5 0 20>, angularLimitsMin=<-5 -80 -10>, angularTarget=<1 0 0>}
      localJointOffset=<2 0 0>
     }]
   }
   n65 = "Constraint - dyn_front_L_skirt_master (P)"(ComponentPose=@n64, Alpha=@n3, ConstraintWeights_0=1) {
    alphaBoolBlend {blendOption=Linear}
    alphaScaleBiasClamp {scale=0.8}
    boneToModify {boneName=dyn_front_L_skirt_master}
    constraintSetup [{targetBone {boneName=thigh_l}}]
    constraintWeights [1]
   }
   n66 = "Constraint - dyn_front_L_skirt_2 (P)"(ComponentPose=@n65, Alpha=@n3, ConstraintWeights_0=1) {
    alphaBoolBlend {blendOption=Linear}
    alphaScaleBiasClamp {scale=0.8}
    boneToModify {boneName=dyn_front_L_skirt_2}
    constraintSetup [{targetBone {boneName=thigh_l}}]
    constraintWeights [1]
   }
   n67 = "Constraint - dyn_front_L_skirt_3 (P)"(ComponentPose=@n66, Alpha=@n3, ConstraintWeights_0=1) {
    alphaBoolBlend {blendOption=Linear}
    alphaScaleBiasClamp {scale=0.8}
    boneToModify {boneName=dyn_front_L_skirt_3}
    constraintSetup [{targetBone {boneName=thigh_l}}]
    constraintWeights [1]
   }
   n68 = "Constraint - dyn_front_R_skirt_master (P)"(ComponentPose=@n67, Alpha=@n2, ConstraintWeights_0=1) {
    alphaBoolBlend {blendOption=Linear}
    alphaScaleBiasClamp {scale=0.8}
    boneToModify {boneName=dyn_front_R_skirt_master}
    constraintSetup [{targetBone {boneName=thigh_r}}]
    constraintWeights [1]
   }
   n69 = "Constraint - dyn_front_R_skirt_2 (P)"(ComponentPose=@n68, Alpha=@n2, ConstraintWeights_0=1) {
    alphaBoolBlend {blendOption=Linear}
    alphaScaleBiasClamp {scale=0.8}
    boneToModify {boneName=dyn_front_R_skirt_2}
    constraintSetup [{targetBone {boneName=thigh_r}}]
    constraintWeights [1]
   }
   n70 = "Constraint - dyn_front_R_skirt_3 (P)"(ComponentPose=@n69, Alpha=@n2, ConstraintWeights_0=1) {
    alphaBoolBlend {blendOption=Linear}
    alphaScaleBiasClamp {scale=0.8}
    boneToModify {boneName=dyn_front_R_skirt_3}
    constraintSetup [{targetBone {boneName=thigh_r}}]
    constraintWeights [1]
   }
   n71 = "Transform (Modify) Bone - Bone: dyn_front_L_skirt_2"(ComponentPose=@n70, Alpha=Get L_leg constraintAlpha(), Translation="0,-2,1", Rotation="-7,-5,0", Scale="1,1.3,1") {
    rotationMode=BMM_Additive
    rotationSpace=BCS_BoneSpace
    scaleMode=BMM_Additive
    scaleSpace=BCS_BoneSpace
    translationMode=BMM_Additive
    translationSpace=BCS_BoneSpace
    alphaBoolBlend {blendOption=Linear}
    boneToModify {boneName=dyn_front_L_skirt_2}
    rotation {pitch=-7, yaw=-5}
    scale {y=1.3}
    translation {y=-2, z=1}
   }
   n72 = "Transform (Modify) Bone - Bone: dyn_front_R_skirt_2"(ComponentPose=@n71, Alpha=Get R_leg constraintAlpha(), Translation="0,2,1", Rotation="-7,5,0", Scale="1,1.3,1") {
    rotationMode=BMM_Additive
    rotationSpace=BCS_BoneSpace
    scaleMode=BMM_Additive
    scaleSpace=BCS_BoneSpace
    translationMode=BMM_Additive
    translationSpace=BCS_BoneSpace
    alphaBoolBlend {blendOption=Linear}
    boneToModify {boneName=dyn_front_R_skirt_2}
    rotation {pitch=-7, yaw=5}
    scale {y=1.3}
    translation {y=2, z=1}
   }
   n73 = "Transform (Modify) Bone - Bone: dyn_side_L_skirt_master"(ComponentPose=@n72, Alpha=Get L_leg constraintAlpha(), Translation="0,0,0", Rotation="-12,0,-12", Scale="1,1,1") {
    rotationMode=BMM_Additive
    rotationSpace=BCS_BoneSpace
    scaleMode=BMM_Additive
    scaleSpace=BCS_BoneSpace
    translationMode=BMM_Additive
    translationSpace=BCS_BoneSpace
    alphaBoolBlend {blendOption=Linear}
    boneToModify {boneName=dyn_side_L_skirt_master}
    rotation {pitch=-12, roll=-12}
   }
   n74 = "Transform (Modify) Bone - Bone: dyn_side_R_skirt_master"(ComponentPose=@n73, Alpha=Get R_leg constraintAlpha(), Translation="0,0,0", Rotation="-12,0,12", Scale="1,1,1") {
    rotationMode=BMM_Additive
    rotationSpace=BCS_BoneSpace
    scaleMode=BMM_Additive
    scaleSpace=BCS_BoneSpace
    translationMode=BMM_Additive
    translationSpace=BCS_BoneSpace
    alphaBoolBlend {blendOption=Linear}
    boneToModify {boneName=dyn_side_R_skirt_master}
    rotation {pitch=-12, roll=12}
   }
   n75 = Get L_leg constraintAlpha()
   n76 = "Constraint - dyn_back_L_skirt_master (P)"(ComponentPose=@n74, Alpha=@n75, ConstraintWeights_0=1) {
    alphaBoolBlend {blendOption=Linear}
    alphaScaleBiasClamp {scale=0.2}
    boneToModify {boneName=dyn_back_L_skirt_master}
    constraintSetup [{targetBone {boneName=thigh_l}}]
    constraintWeights [1]
   }
   n77 = "Constraint - dyn_back_L_skirt_2 (P)"(ComponentPose=@n76, Alpha=@n75, ConstraintWeights_0=1) {
    alphaBoolBlend {blendOption=Linear}
    alphaScaleBiasClamp {scale=0.2}
    boneToModify {boneName=dyn_back_L_skirt_2}
    constraintSetup [{targetBone {boneName=thigh_l}}]
    constraintWeights [1]
   }
   n78 = "Constraint - dyn_back_L_skirt_3 (P)"(ComponentPose=@n77, Alpha=@n75, ConstraintWeights_0=1) {
    alphaBoolBlend {blendOption=Linear}
    alphaScaleBiasClamp {scale=0.2}
    boneToModify {boneName=dyn_back_L_skirt_3}
    constraintSetup [{targetBone {boneName=thigh_l}}]
    constraintWeights [1]
   }
   n79 = "Constraint - dyn_back_R_skirt_master (P)"(ComponentPose=@n78, Alpha=@n1, ConstraintWeights_0=1) {
    alphaBoolBlend {blendOption=Linear}
    alphaScaleBiasClamp {scale=0.2}
    boneToModify {boneName=dyn_back_R_skirt_master}
    constraintSetup [{targetBone {boneName=thigh_r}}]
    constraintWeights [1]
   }
   n80 = "Constraint - dyn_back_R_skirt_2 (P)"(ComponentPose=@n79, Alpha=@n1, ConstraintWeights_0=1) {
    alphaBoolBlend {blendOption=Linear}
    alphaScaleBiasClamp {scale=0.2}
    boneToModify {boneName=dyn_back_R_skirt_2}
    constraintSetup [{targetBone {boneName=thigh_r}}]
    constraintWeights [1]
   }
   n81 = "Constraint - dyn_back_R_skirt_3 (P)"(ComponentPose=@n80, Alpha=@n1, ConstraintWeights_0=1) {
    alphaBoolBlend {blendOption=Linear}
    alphaScaleBiasClamp {scale=0.2}
    boneToModify {boneName=dyn_back_R_skirt_3}
    constraintSetup [{targetBone {boneName=thigh_l}}]
    constraintWeights [1]
   }
   n82 = "Transform (Modify) Bone - Bone: dyn_coilBelt_A"(ComponentPose=@n81, Alpha=Get eitherLegUpAlpha() -> (eitherLegUpAlpha:double), Translation="2,0,6.5", Rotation="0,0,0", Scale="1,1,0.7") {
    scaleMode=BMM_Additive
    scaleSpace=BCS_BoneSpace
    translationMode=BMM_Additive
    translationSpace=BCS_BoneSpace
    alphaBoolBlend {blendOption=Linear}
    boneToModify {boneName=dyn_coilBelt_A}
    scale {z=0.7}
    translation {x=2, z=6.5}
   }
   n83 = Get L leg up() -> (L leg up:double)
   n84 = "Transform (Modify) Bone - Bone: dyn_coilBelt_J"(ComponentPose=@n82, Alpha=@n83, Translation="0.2,-0.2,6.8", Rotation="0,0,0", Scale="1,1,1") {
    translationMode=BMM_Additive
    translationSpace=BCS_BoneSpace
    alphaBoolBlend {blendOption=Linear}
    boneToModify {boneName=dyn_coilBelt_J}
    translation=<0.2 -0.2 6.8>
   }
   n85 = "Transform (Modify) Bone - Bone: dyn_coilBelt_I"(ComponentPose=@n84, Alpha=@n83, Translation="0.8,-1.2,3.6", Rotation="0,0,0", Scale="1,1,1") {
    translationMode=BMM_Additive
    translationSpace=BCS_BoneSpace
    alphaBoolBlend {blendOption=Linear}
    boneToModify {boneName=dyn_coilBelt_I}
    translation=<0.8 -1.2 3.6>
   }
   n86 = Get R leg up() -> (R leg up:double)
   n87 = "Transform (Modify) Bone - Bone: dyn_coilBelt_B"(ComponentPose=@n85, Alpha=@n86, Translation="0,2,7", Rotation="0,0,0", Scale="1,1,1") {
    translationMode=BMM_Additive
    translationSpace=BCS_BoneSpace
    alphaBoolBlend {blendOption=Linear}
    boneToModify {boneName=dyn_coilBelt_B}
    translation {y=2, z=7}
   }
   n88 = "Transform (Modify) Bone - Bone: dyn_coilBelt_C"(ComponentPose=@n87, Alpha=@n86, Translation="1,1,4", Rotation="0,0,0", Scale="1,1,1") {
    translationMode=BMM_Additive
    translationSpace=BCS_BoneSpace
    alphaBoolBlend {blendOption=Linear}
    boneToModify {boneName=dyn_coilBelt_C}
    translation=<1 1 4>
   }
   n89 = "Constraint - dyn_back_C_skirt_1 (PP)"!(ComponentPose=@n88, Alpha=1, ConstraintWeights_0=0.5, ConstraintWeights_1:float=0.5) -> (Pose) {
    alphaBoolBlend {blendOption=Linear}
    alphaScaleBiasClamp {interpSpeedDecreasing=20, interpSpeedIncreasing=20, scale=0.3}
    boneToModify {boneName=dyn_back_C_skirt_1}
    constraintSetup [{targetBone {boneName=dyn_back_L_skirt_1}}, {targetBone {boneName=dyn_back_R_skirt_1}}]
    constraintWeights [1,1]
   }
   n90 = "Constraint - dyn_back_C_skirt_2 (PP)"!(ComponentPose=@n89, Alpha=1, ConstraintWeights_0=0.5, ConstraintWeights_1:float=0.5) -> (Pose) {
    alphaBoolBlend {blendOption=Linear}
    alphaScaleBiasClamp {interpSpeedDecreasing=20, interpSpeedIncreasing=20, scale=0.3}
    boneToModify {boneName=dyn_back_C_skirt_2}
    constraintSetup [{targetBone {boneName=dyn_back_L_skirt_2}}, {targetBone {boneName=dyn_back_R_skirt_2}}]
    constraintWeights [1,1]
   }
   n91 = "Constraint - dyn_back_C_skirt_3 (PP)"!(ComponentPose=@n90, Alpha=1, ConstraintWeights_0=0.5, ConstraintWeights_1:float=0.5) -> (Pose) {
    alphaBoolBlend {blendOption=Linear}
    alphaScaleBiasClamp {interpSpeedDecreasing=20, interpSpeedIncreasing=20, scale=0.3}
    boneToModify {boneName=dyn_back_C_skirt_3}
    constraintSetup [{targetBone {boneName=dyn_back_L_skirt_3}}, {targetBone {boneName=dyn_back_R_skirt_3}}]
    constraintWeights [1,1]
   }
   n92 = "Constraint - dyn_back_L_skirt_master (P)"(ComponentPose=@n91, Alpha=Get L leg back() -> (L leg back:double), ConstraintWeights_0=1) {
    alphaBoolBlend {blendOption=Linear}
    alphaScaleBiasClamp {bInterpResult=true, interpSpeedDecreasing=45, interpSpeedIncreasing=45}
    boneToModify {boneName=dyn_back_L_skirt_master}
    constraintSetup [{targetBone {boneName=thigh_l}}]
    constraintWeights [1]
   }
   n93 = "Constraint - dyn_back_R_skirt_master (P)"(ComponentPose=@n92, Alpha=Get R leg back() -> (R leg back:double), ConstraintWeights_0=1) {
    alphaBoolBlend {blendOption=Linear}
    alphaScaleBiasClamp {bInterpResult=true, interpSpeedDecreasing=45, interpSpeedIncreasing=45}
    boneToModify {boneName=dyn_back_R_skirt_master}
    constraintSetup [{targetBone {boneName=thigh_r}}]
    constraintWeights [1]
   }
   Output Pose(Result=Component To Local(ComponentPose=@n93)) {layerGroup=None}
  ]}
 Calculate Cloth Sim Noise {nodes [
   n1 = Calculate Cloth Sim Noise!() -> (
    then
    MID:Material Instance Dynamic Object Reference
    OverlayMID:Material Instance Dynamic Object Reference
    SkelMesh:Skeletal Mesh Component Object Reference
    BoneName:Name=root
    DeltaTime:double=~
    NoiseMovementStrength:double=0.25
    NoiseBoneLocation:"Vector (by ref)"
    NoiseVelocityVector:"Vector (by ref)"
   ) {is=title K2Node_FunctionEntry}
   > n2 = calculateMotionVectorForSocket!(
    execute
    self
    skelMesh:Scene Component Object Reference=@n1.SkelMesh
    boneName:Name=@n1.BoneName
    previousBoneLocation:Vector=@n1.NoiseBoneLocation
   ) -> (then, boneMotionVector:Vector, boneLocation:Vector, boneTransform:Transform)
   n3 = vector + vector(
    A=vector * vector(A=@n2, B=float * float(A=@n1.DeltaTime, B=@n1.NoiseMovementStrength) {is=title K2Node_PromotableOperator}) -> (ReturnValue=~)
    B=@n1.NoiseVelocityVector
   )
   > SetVectorParameterValue(
    self=[@n1.OverlayMID, @n1]
    ParameterName=ClothSim_ABP_TurbulenceUVOffset
    Value="To LinearColor (Vector)"!(InVec:Vector=@n3) -> (ReturnValue:Linear Color Structure)
   )
   > "Set Vector (by ref)"!(execute:Exec, Target:"Vector (by ref)"=@n1.NoiseBoneLocation, Value:Vector=@n2.boneLocation) -> (then:Exec) {is=title K2Node_VariableSetRef}
   > "Set Vector (by ref)"(Target=@n1.NoiseVelocityVector, Value=@n3)
   > Return Node()
  ]}
 EventGraph {nodes [
   Event BlueprintUpdateAnimation()
   > n1 = Sequence!(execute) -> (then_0, then_1, then_2:Exec)
   n2 = GetOwnerSkeletalMeshComponent()
   n3 = BreakRotator(InRot=BreakTransform(InTransform=GetDeltaTransformFromRefPose(self=@n2, BoneName=thigh_l, BaseName=pelvis)).Rotation)
   Set L_leg constraintAlpha(
    execute=@n1.then_1
    L_leg constraintAlpha:double="Max (Float)"(A=MapRangeClamped(Value=@n3.Yaw, InRangeA=-30, InRangeB=-60), B=MapRangeClamped(Value=@n3.Pitch, InRangeA=5, InRangeB=-10))
   ) -> (Output_Get:double)
   > Set L leg back(L leg back:double=MapRangeClamped(Value=@n3.Yaw, InRangeA=-30, InRangeB=-8)) -> (Output_Get:double)
   n4 = BreakRotator(InRot=BreakTransform(InTransform=GetDeltaTransformFromRefPose(self=@n2, BoneName=thigh_r, BaseName=pelvis)).Rotation)
   > Set R_leg constraintAlpha(R_leg constraintAlpha:double="Max (Float)"(A=MapRangeClamped(Value=@n4.Yaw, InRangeA=-15, InRangeB=-60), B=MapRangeClamped(Value=@n4.Pitch, InRangeA=5, InRangeB=-10))) -> (Output_Get:double)
   > Set R leg back(R leg back:double=MapRangeClamped(Value=@n4.Yaw, InRangeA=-30, InRangeB=-8)) -> (Output_Get:double)
   > Set eitherLegConstraintAlpha(eitherLegConstraintAlpha:double="Max (Float)"(A=Get L_leg constraintAlpha(), B=Get R_leg constraintAlpha())) -> (Output_Get:double)
   n5 = BreakTransform(InTransform=GetDeltaTransformFromRefPose(self=@n2, BoneName=thigh_l, BaseName=pelvis))
   > Set L leg up(L leg up:double=MapRangeClamped(Value=BreakRotator(InRot=@n5.Rotation).Yaw, InRangeA=-70, InRangeB=-110)) -> (Output_Get:double)
   n6 = BreakTransform(InTransform=GetDeltaTransformFromRefPose(self=@n2, BoneName=thigh_r, BaseName=pelvis))
   > Set R leg up(R leg up:double=MapRangeClamped(Value=BreakRotator(InRot=@n6.Rotation).Yaw, InRangeA=-70, InRangeB=-110)) -> (Output_Get:double)
   > Set eitherLegUpAlpha(eitherLegUpAlpha:double="Max (Float)"(A=Get L leg up(), B=Get R leg up())) -> (Output_Get:double)
   Set skydiveGravity(execute=@n1.then_2, skydiveGravity:Vector=SelectVector(A="0,0,200", B="0,0,-1500", bPickA=Get bIsSkyDiving())) -> (Output_Get:Vector)
   > n7 = Branch(Condition=Get bIsDBNOCarried())
   > Set simSpaceChange(simSpaceChange:AnimPhysSimSpaceType Enum=Component) -> (Output_Get:AnimPhysSimSpaceType Enum)
   Set simSpaceChange(execute=@n7.else)
   n8 = Sequence!(execute=@n1.then_0) -> (then_0, then_1, then_2, then_3:Exec, then_4:Exec)
   n9 = Get bIsSkyDiving()
   > n10 = Branch(Condition=@n9)
   > n11 = Set ClothSim_WindStrength(ClothSim_WindStrength:double=0.4) -> (Output_Get:double)
   n12 = Get Mat Instance Dynamic()
   SetScalarParameterValue(
    execute=[@n13.then, @n11.then]
    self=@n12
    ParameterName=ClothSim_WindStrength
    Value=Get ClothSim_WindStrength() -> (ClothSim_WindStrength:double)
   )
   n13 = Set ClothSim_WindStrength(execute=@n10.else, ClothSim_WindStrength=0)
   n14 = Branch(execute=@n8.then_1, Condition=@n9)
   > n15 = Set ClothSim_VelocityMax(ClothSim_VelocityMax:double=750) -> (Output_Get:double)
   SetScalarParameterValue(
    execute=[@n16.then, @n15.then]
    self=@n12
    ParameterName=ClothSim_VelocityMax
    Value=Get ClothSim_VelocityMax() -> (ClothSim_VelocityMax:double)
   )
   n16 = Set ClothSim_VelocityMax(execute=@n14.else, ClothSim_VelocityMax=5000)
   n17 = Branch(execute=@n8.then_2, Condition=@n9)
   > n18 = Set ClothSim_WindNoiseSpeed(ClothSim_WindNoiseSpeed:double=200) -> (Output_Get:double)
   SetScalarParameterValue(
    execute=[@n19.then, @n18.then]
    self=@n12
    ParameterName=ClothSim_WindNoiseSpeed
    Value=Get ClothSim_WindNoiseSpeed() -> (ClothSim_WindNoiseSpeed:double)
   )
   n19 = Set ClothSim_WindNoiseSpeed(execute=@n17.else, ClothSim_WindNoiseSpeed=0)
   n20 = Branch(execute=@n8.then_4, Condition=Get movementSum() -> (movementSum:bool))
   SetVectorParameterValue(
    execute=@n20.else
    self=Get Mat Instance Dynamic()
    ParameterName=ClothSim_ABP_R1_Velocity
    Value=Get Value() -> (Value:Linear Color Structure)
   )
   Set WPO Globals Function(
    execute=@n8.then_3
    Material=Get Mat Instance Dynamic()
    Overlay Material=Get Overlay Mat Instance Dynamic()
    Skeletal Mesh=Get skeletal_mesh() -> (skeletal_mesh:Skeletal Mesh Component Object Reference)
    "Local Noise Transform  Location"=Get Local Noise Transform Location() -> (Local Noise Transform Location:Vector)
    Local Noise Offset=Get Local Noise Offset()
   )
   > Set WPO Regional Function!(
    execute
    Material:Material Instance Dynamic Object Reference=Get Mat Instance Dynamic()
    Overlay Material:Material Instance Dynamic Object Reference=Get Overlay Mat Instance Dynamic()
    Skeletal Mesh:Skeletal Mesh Component Object Reference=Get skeletal_mesh()
    Velocity Transform:Name=pelvis
    Velocity Interp Speed:double=Get Velocity Interp Speed()
    Collision Transform:Name=spine_01
    Collision Transform Axis:EAxis Enum
    Invert Collision Transform Axis:bool=false
    Delta Time X:double=Reroute Node(InputPin:double) -> (OutputPin:double)
    Previous Velocity Transform Location:"Vector (by ref)"=Get R1 Velocity Transform Location()
    Previous Velocity Transform Velocity:"Vector (by ref)"=Get R1 Velocity Transform Velocity()
    Region:int=1
   ) -> (then)
   > Get!(execute:Exec) -> (then:Exec, else:Exec, ownerSkelMesh:Skeletal Mesh Component Object Reference)
   > Get!(execute) -> (then, else, PlayerPawn:Fort Player Pawn Athena Object Reference)
   > n21 = PlayerMovement(AnimInstance=Self-Reference(), XDotProductCoefficient=1, YDotProductCoefficient=1, ZDotProductCoefficient=1)
   > n22 = Branch(Condition="Not Equal (Float)"(A=float + float!(A=@n21, B=@n21.YMovement, C:float=@n21.ZMovement) -> (ReturnValue)))
   > Set movementSum(movementSum:bool=true) -> (Output_Get:bool)
   Set movementSum(execute=@n22.else, movementSum=false)
   Event OnCharacterCustomizationLoaded()
   > n23 = Sequence!(execute) -> (then_0, then_1)
   n24 = Set skeletal_mesh(execute=@n23.then_1, skeletal_mesh:Skeletal Mesh Component Object Reference=GetPartSkeletalMeshComponent()) -> (Output_Get:Skeletal Mesh Component Object Reference)
   > n25 = Get Dynamic Material Instance(SkeletalMeshComponent=@n24)
   > Set Mat Instance Dynamic(Mat Instance Dynamic=@n25)
   > n26 = GetOrCreateDynamicOverlayMaterial(MeshComponent=@n24, MaterialName=None)
   > Set Overlay Mat Instance Dynamic(Overlay Mat Instance Dynamic=@n26)
   n27 = CheckForBackbling()
   Branch(execute=@n23.then_0, Condition=@n27)
   > n28 = Branch(Condition=@n27.bIsBackblingACape)
   > Set Is Backbling A Cape(Is Backbling A Cape:bool=@n27.bIsBackblingACape) -> (Output_Get:bool)
   Set Is Wearing Backbling(execute=@n28.else, Is Wearing Backbling=@n27)
   OR Boolean(A=Get bIsInFrontEnd(), B=false)
  ]}
 GetBoneNames {nodes [
   n1 = GetBoneNames!() -> (then, Region:"Integer (by ref)") {is=title K2Node_FunctionEntry}
   n2 = Append(A=ClothSim_ABP_R, B="To String (Integer)"(InInt=@n1))
   > Return Node!(
    execute
    Velocity:Name=String To Name(InString=Append(A=@n2, B=_Velocity))
    BonePosition:Name=String To Name(InString=Append(A=@n2, B=_BonePosition))
    BoneVector:Name=String To Name(InString=Append(A=@n2, B=_BoneVector))
   )
  ]}
 Set Cloth Sim Params {nodes [
   n1 = Set Cloth Sim Params!() -> (
    then
    material:Material Instance Dynamic Object Reference
    overlayMaterial:Material Instance Dynamic Object Reference
    skelMesh:Skeletal Mesh Component Object Reference
    boneName:Name
    interpSpeed:double=2
    collisionBoneName:Name
    collisionBoneAxis:EAxis Enum=None
    invertCollisionBoneAxis:bool=~
    deltaTimeX:double=~
    previousBoneLocation:"Vector (by ref)"
    previousBoneVelocityVector:"Vector (by ref)"
    region:int=1
   ) {is=title K2Node_FunctionEntry}
   > Set lMID(lMID:Material Instance Dynamic Object Reference=@n1) -> (Output_Get:Material Instance Dynamic Object Reference)
   > Set loverlayMID(loverlayMID:Material Instance Dynamic Object Reference=@n1.overlayMaterial) -> (Output_Get:Material Instance Dynamic Object Reference)
   > Set lSkelMesh(lSkelMesh:Skeletal Mesh Component Object Reference=@n1.skelMesh) -> (Output_Get:Skeletal Mesh Component Object Reference)
   > n2 = GetBoneNames!(execute, self, Region:"Integer (by ref)"=@n1.region) -> (then, Velocity:Name=None, BonePosition:Name=None, BoneVector:Name=None) {is=title K2Node_CallFunction}
   > Set lMatVelocityParamName(lMatVelocityParamName:Name=@n2) -> (Output_Get:Name=None)
   > Set lMatBonePositionParamName(lMatBonePositionParamName:Name=@n2.BonePosition) -> (Output_Get:Name=None)
   > Set lMatBoneVectorParamName(lMatBoneVectorParamName:Name=@n2.BoneVector) -> (Output_Get:Name=None)
   > n3 = calculateMotionVectorForSocket(
    skelMesh=Get lSkelMesh() -> (lSkelMesh:Skeletal Mesh Component Object Reference)
    boneName=@n1.boneName
    previousBoneLocation=@n1.previousBoneLocation
   )
   n4 = VInterpTo!(
    Current:Vector=@n1.previousBoneVelocityVector
    Target:Vector=vector / vector(A:Vector=@n3, B=@n1.deltaTimeX) -> (ReturnValue:Vector)
    DeltaTime:float=@n1.deltaTimeX
    InterpSpeed:float=@n1.interpSpeed
   ) -> (ReturnValue:Vector)
   > SetVectorParameterValue(
    self=[
     Get loverlayMID() -> (loverlayMID:Material Instance Dynamic Object Reference)
     Get lMID() -> (lMID:Material Instance Dynamic Object Reference)
    ]
    ParameterName=Get lMatVelocityParamName() -> (lMatVelocityParamName:Name=None)
    Value="To LinearColor (Vector)"(InVec=@n4)
   )
   n5 = GetSocketTransform(self=Get lSkelMesh(), InSocketName=@n1.collisionBoneName, TransformSpace=RTS_World)
   n6 = "Get Unit Axis (Matrix)"!(
    M:"Matrix Structure (by ref)"="To Matrix (Transform)"!(Transform:"Transform (by ref)"=@n5) -> (ReturnValue:Matrix Structure)
    Axis:EAxis Enum=@n1.collisionBoneAxis
   ) -> (ReturnValue:Vector) {is=title K2Node_CallFunction}
   n7 = Select(Option 0:double=1, Option 1:double=-1, Index=@n1.invertCollisionBoneAxis) -> (ReturnValue:double)
   > SetVectorParameterValue(
    self=[Get loverlayMID(), Get lMID()]
    ParameterName=Get lMatBoneVectorParamName() -> (lMatBoneVectorParamName:Name=None)
    Value="To LinearColor (Vector)"(InVec=vector * vector(A=@n6, B=@n7) -> (ReturnValue=~))
   )
   > SetVectorParameterValue(
    self=[Get loverlayMID(), Get lMID()]
    ParameterName=Get lMatBonePositionParamName() -> (lMatBonePositionParamName:Name=None)
    Value="To LinearColor (Vector)"(InVec=BreakTransform(InTransform=@n5))
   )
   > "Set Vector (by ref)"(Target=@n1.previousBoneLocation, Value=@n3.boneLocation)
   > "Set Vector (by ref)"(Target=@n1.previousBoneVelocityVector, Value=@n4)
   > Return Node!(execute)
   Set lMID()
  ]}
 calculateMotionVectorForSocket {nodes [
   n1 = calculateMotionVectorForSocket!() -> (then, skelMesh:Scene Component Object Reference, boneName:Name, previousBoneLocation:Vector=~) {is=title K2Node_FunctionEntry}
   n2 = GetSocketTransform(self=@n1, InSocketName=@n1.boneName, TransformSpace=RTS_World)
   n3 = BreakTransform(InTransform=@n2)
   > Return Node!(execute, boneMotionVector:Vector=vector - vector(A=@n3, B=@n1.previousBoneLocation), boneLocation:Vector=@n3, boneTransform:Transform=@n2)
  ]}
 setWPOMaterialVelocity {nodes [
   n1 = setWPOMaterialVelocity!() -> (
    then
    materialInstance:Material Instance Dynamic Object Reference
    skelMesh:Skeletal Mesh Component Object Reference
    boneName:Name
    offsetBoneName:Name
    offsetBoneAxis:EAxis Enum=None
    deltaTimeX:double=~
    materialVelocityAttribute:Name=Velocity
    previousBoneLocation:Vector=~
    previousBoneVelocityVector:Vector=~
   ) {is=title K2Node_FunctionEntry}
   > n2 = calculateMotionVectorForSocket(skelMesh=@n1.skelMesh, boneName=@n1.boneName, previousBoneLocation=@n1.previousBoneLocation) {is=title K2Node_CallFunction}
   n3 = "Get Unit Axis (Matrix)"(
    M="To Matrix (Transform)"(Transform=GetSocketTransform(self=@n1.skelMesh, InSocketName=@n1.offsetBoneName, TransformSpace=RTS_World))
    Axis=@n1.offsetBoneAxis
   )
   n4 = vector / vector(A=@n2, B=@n1.deltaTimeX)
   n5 = VInterpTo(Current=@n1.previousBoneVelocityVector, Target=@n4, DeltaTime=@n1.deltaTimeX, InterpSpeed=0.75)
   > SetVectorParameterValue(self=@n1, ParameterName=@n1.materialVelocityAttribute, Value="To LinearColor (Vector)"(InVec=vector + vector(A=@n3, B=@n5)))
   > Return Node!(
    execute
    boneVelocity:double=Vector Length(A=@n4)
    boneVelocityVector:Vector=@n4
    boneLocation=@n2.boneLocation
    boneTransform=@n2.boneTransform
   )
  ]}
]
variables [
 ClothSim_VelocityMax {type=number}
 ClothSim_WindNoiseSpeed {type=number}
 ClothSim_WindStrength {type=number}
 eitherLegConstraintAlpha {type=number}
 eitherLegUpAlpha {type=number}
 Gravity Override Params {
  type=GravityOverrideParamsStruct
  default {
   skydive_DivingDown {gravityOverride=<-1500 0 100>}
   skydive_FloatingDown {gravityOverride=<0 -700 100>}
   skydive_Parachuting {gravityOverride=<0 0 -1500>}
  }
 }
 Is Backbling A Cape {type=boolean}
 Is Wearing Backbling {type=boolean}
 L leg back {type=number}
 L leg up {type=number}
 L_leg constraintAlpha {type=number}
 Local Noise Offset {type=Vector}
 Local Noise Transform Location {type=Vector}
 Mat Instance Dynamic {type=MaterialInstanceDynamic}
 movementSum {type=boolean}
 Overlay Mat Instance Dynamic {type=MaterialInstanceDynamic}
 ownerSkelMesh {type=SkeletalMeshComponent}
 PlayerPawn {type=FortPlayerPawnAthena}
 R leg back {type=number}
 R leg up {type=number}
 R1 Velocity Transform Location {type=Vector}
 R1 Velocity Transform Velocity {type=Vector}
 R_leg constraintAlpha {type=number}
 sideBlingConstraint {type=number}
 simSpaceChange {type=AnimPhysSimSpaceType}
 skeletal_mesh {type=SkeletalMeshComponent}
 skydiveGravity {type=Vector}
 Value {type=LinearColor, default=<rgba 0 0 0 0>}
 Velocity Interp Speed {type=number}
]
