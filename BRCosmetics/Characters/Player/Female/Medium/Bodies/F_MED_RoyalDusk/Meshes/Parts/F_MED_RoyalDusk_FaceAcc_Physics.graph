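# Physics asset for the face accessory mesh of the RoyalDusk female medium player cosmetic, simulating its dynamic hair.
# A limited-swing chain runs from the head through three dyn_hairMain bodies, then branches into looser C, D and E tip bodies.
# Capsules on the head, clavicles and upper spine, linked by free angular constraints, form the body volumes around the hair.
# All shapes are capsules, with the head reduced to a tiny one.
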PhysicsAsset /BRCosmetics/Characters/Player/Female/Medium/Bodies/F_MED_RoyalDusk/Meshes/Parts/F_MED_RoyalDusk_FaceAcc_Physics
bodies {
 clavicle_l [{shapeType=Capsule, length=2.6527, radius=8.3825, shapeName=clavicle_l_capsule, center=<6.5025 0.308 -2.2688>, rotation=<rot 85 0 0>}]
 clavicle_r [{shapeType=Capsule, length=2.6527, radius=8.3825, shapeName=clavicle_l_capsule, center=<-6.5025 -0.308 2.2688>, rotation=<rot 85 0 0>}]
 dyn_hairMain_1 [{shapeType=Capsule, length=1, radius=3.5, shapeName=dyn_hairMain_1_capsule, center=<3 0 0>, rotation=<rot 90 0 0>}]
 dyn_hairMain_2 [{shapeType=Capsule, length=1, radius=4.5, shapeName=dyn_hairMain_2_capsule, center=<7 0 0>, rotation=<rot 95 0 0>}]
 dyn_hairMain_3 [{shapeType=Capsule, length=2, radius=4, shapeName=dyn_hairMain_3_capsule, center=<7 0 0>, rotation=<rot 90 0 0>}]
 dyn_hairMain_C [{shapeType=Capsule, length=5.5, radius=2.5, shapeName=dyn_hairMain_C_capsule, center=<2 1 0>, rotation=<rot 90 0 0>}]
 dyn_hairMain_D [{shapeType=Capsule, length=4, radius=3, shapeName=dyn_hairMain_D_capsule, center=<1.8315 0 0>, rotation=<rot 89.2234 -84.9503 -80.6014>}]
 dyn_hairMain_E [{shapeType=Capsule, length=5.5, radius=2.2, shapeName=dyn_hairMain_C_capsule, center=<2 1.5 0>, rotation=<rot 90 0 0>}]
 head [{shapeType=Capsule, length=1, radius=1, shapeName=head_capsule}]
 spine_04 [{shapeType=Capsule, length=8.0004, radius=10, shapeName=spine_05_capsule, center=<2.3173 0.3523 0>}]
 spine_05 [{shapeType=Capsule, length=9.8891, radius=10, shapeName=spine_05_capsule, center=<2.8219 0.0619 0>}]
}
constraints [
 {bone1Name=dyn_hairMain_2, bone2Name=dyn_hairMain_1, swing1LimitDegrees=20, swing2LimitDegrees=45, twistLimitDegrees=20}
 {bone1Name=dyn_hairMain_3, bone2Name=dyn_hairMain_2, swing1LimitDegrees=20, swing2LimitDegrees=45, twistLimitDegrees=20}
 {bone1Name=dyn_hairMain_C, bone2Name=dyn_hairMain_3, swing1LimitDegrees=60, swing2LimitDegrees=60, twistLimitDegrees=60}
 {bone1Name=dyn_hairMain_D, bone2Name=dyn_hairMain_3, swing1LimitDegrees=60, swing2LimitDegrees=60, twistLimitDegrees=60}
 {bone1Name=dyn_hairMain_1, bone2Name=head, swing1LimitDegrees=20, swing2LimitDegrees=45, twistLimitDegrees=20}
 {
  bone1Name=head
  bone2Name=spine_05
  swing1LimitDegrees=45
  swing1Motion=Free
  swing2LimitDegrees=45
  swing2Motion=Free
  twistLimitDegrees=45
  twistMotion=Free
 }
 {
  bone1Name=clavicle_l
  bone2Name=spine_05
  swing1LimitDegrees=45
  swing1Motion=Free
  swing2LimitDegrees=45
  swing2Motion=Free
  twistLimitDegrees=45
  twistMotion=Free
 }
 {
  bone1Name=clavicle_r
  bone2Name=spine_05
  swing1LimitDegrees=45
  swing1Motion=Free
  swing2LimitDegrees=45
  swing2Motion=Free
  twistLimitDegrees=45
  twistMotion=Free
 }
 {bone1Name=dyn_hairMain_E, bone2Name=dyn_hairMain_3, swing1LimitDegrees=60, swing2LimitDegrees=60, twistLimitDegrees=60}
 {
  bone1Name=spine_05
  bone2Name=spine_04
  swing1LimitDegrees=45
  swing1Motion=Free
  swing2LimitDegrees=45
  swing2Motion=Free
  twistLimitDegrees=45
  twistMotion=Free
 }
]
